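; Reusable Creative device component that holds a team selection and moves players between minigame teams.
; Initialization maps the team knob, legacy or new encoding, to a selector state and desired team, deferring delegate binding until a minigame exists.
; Play-mode changes reapply the team within the team limit, and its RepNotify broadcasts updates and can retag the owning building.
; Server-side switching picks a fixed or different random team, clears items and broadcasts success.
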(event EventEndPlay
  (params (EndPlayReason :type EEndPlayReason%20Enum :default "Destroyed"))
  (Initialization|BindDelegates false))

(fn AttemptPlayerTeamSwitch
  (params (Controller :type Controller%20Object%20Reference) (bChangeTeamWithoutRespawn :type Boolean) (bResetVitalsOnSwitch :type Boolean) (Clear%20Items%20on%20Switch :type Boolean))
  (bind (_valid_number_of_teams _output_team) (Getters|GetMaxTeams))
  (bind _localminigame (LocalVariables|GetLocalMinigame))
  (bind _returnvalue (Minigame|GetMinigameTeamComponent _localminigame))
  (bind _local_team (LocalVariables|GetLocal%20Team))
  (LocalVariables|SetAllowTestPlayer (Class|CreativeBlueprintLibrary|IsAllowingTestPlayeronCreativeDevices))
  (if (Math|Boolean|ANDBoolean (Networking|IsServer) (Utilities|Enum|NotEqual(Enum) (Variables|Internal|GetTeamSelectorState) "NewEnumerator0") _valid_number_of_teams)
    (LocalVariables|SetLocalNumTeams _output_team)
    (if (LocalVariables|GetAllowTestPlayer)
      (bind _asfort_athena_aibot_controller (Utilities|Casting|CastToFortAthenaAIBotController Controller)
        (:then
          (LocalVariables|SetTestController _asfort_athena_aibot_controller))
        (:CastFailed
          (bind _asfort_player_controller_athena (Utilities|Casting|CastToFortPlayerControllerAthena Controller)
            (:then
              (LocalVariables|SetTestController _asfort_player_controller_athena))
            (:CastFailed
              (return false)))))
      (bind _playerstate (Variables|/Script/Engine.Controller|GetPlayerState (LocalVariables|GetTestController))
        (:then
          (bind _asfort_player_state_athena (Utilities|Casting|CastToFortPlayerStateAthena _playerstate)
            (:then
              (LocalVariables|SetLocalPS _asfort_player_state_athena)
              (bind _currentminigame (Variables|/Script/FortniteGame.FortMinigameLogicComponent|GetCurrentMinigame (Variables|Default|GetLogic%20Component))
                (:then
                  (LocalVariables|SetLocalMinigame _currentminigame)
                  (LocalVariables|SetPrevious%20Team (Team|GetActorPvPTeamIndex (LocalVariables|GetTestController)))
                  (Utilities|FlowControl|Sequence
                    (:then_0
                      ;[1 1 1] Define which team we want to switch to.
                      (if (Utilities|Enum|Equal(Enum) (Variables|Internal|GetTeamSelectorState) "NewEnumerator2")
                        (bind _output_get (LocalVariables|SetResult (+ (Math|Random|RandomIntegerinRange 1 (- (LocalVariables|GetLocalNumTeams) 1)) (Math|Conversions|ToInteger(Byte) (LocalVariables|GetPrevious%20Team)))))
                        (LocalVariables|SetLocal%20Team (Math|Integer|%25%28Integer%29 _output_get (LocalVariables|GetLocalNumTeams)))
                        (else
                          (LocalVariables|SetLocal%20Team (Variables|Internal|GetTeamInt)))))
                    (:then_1
                      (if (Math|Byte|NotEqual(Byte) (LocalVariables|GetPrevious%20Team) (Math|Conversions|ToByte(Integer) _local_team))
                        (PlayerTeamSwitch|ClearItemsfromPlayer (LocalVariables|GetTestController))
                        (Utilities|IsValid _returnvalue
                          (:Is%20Valid
                            (bind _returnvalue_2 (Minigame|SetTeamforPlayer _returnvalue (LocalVariables|GetLocalPS) (Teams|IndextoTeam _local_team) true (Variables|GetbChangeTeamWithoutRespawn) (Variables|GetbResetVitalsOnSwitch) (select (Variables|GetClear%20Items%20on%20Switch) "ClearOnTeamSwitch" "NeverClear") (Variables|Tags|GetContextTag)))
                            (if _returnvalue_2
                              (Default|CallOnPlayerTeamSwitched (LocalVariables|GetTestController) (LocalVariables|GetLocal%20Team) (Math|Conversions|ToInteger(Byte) (LocalVariables|GetPrevious%20Team)))
                              (return true)
                              (else
                                _)))
                          (:Is%20Not%20Valid
                            _))
                        (return false)
                        (else
                          (return false))))))
                (:else
                  (return false))))
            (:CastFailed
              (return false))))
        (:else
          (return false)))
      (else
        _))
    (else
      (return false))))

(fn AttemptSetTeam
  (params (New%20Team :type Integer))
  (bind (_valid_number_of_teams _output_team) (Getters|GetMaxTeams))
  (bind _logic_component (Variables|Default|GetLogic%20Component)
    (:then
      (Variables|/Script/FortniteGame.FortMinigameLogicComponent|GetCurrentMinigame _logic_component
        (:then
          (Variables|Internal|SetTeamInt (select (< New%20Team _output_team) New%20Team 0))
          (Networking|FlushNetDormancy (Components|GetOwner)))
        (:else)) ; Only want to set the Team if we have a valid minigame\n(avoids unnecessary replication), as if we don't have a valid \nminigame it will set when the minigame is valid.
    )
    (:else)))

(fn BindDelegates
  (params (Bind :type Boolean :default true))
  (bind _outputdelegate (EventDispatchers|CreateEvent :!event Set%20Team%20on%20Play%20Mode%20Change))
  (bind _logic_component (Variables|Default|GetLogic%20Component)
    (:then
      (bind _currentminigame (Variables|/Script/FortniteGame.FortMinigameLogicComponent|GetCurrentMinigame _logic_component)
        (:then
          (if Bind
            (Minigame|BindEventtoOnPlayModeChangedDelegate _currentminigame _outputdelegate)
            (else
              (Minigame|UnbindEventfromOnPlayModeChangedDelegate _currentminigame _outputdelegate)))
          (return true))
        (:else
          ;[1 1 1] Minigame isn't valid yet, reattempt this when it is.
          (Minigame|BindEventtoOnAddedtoMinigame _logic_component (EventDispatchers|CreateEvent :!event Reattempt%20Post-Minigame%20Add%20Initialization))
          (return false))))
    (:else
      (return false))))

(fn ClearItemsFromPlayer
  (params (Controller :type Controller%20Object%20Reference))
  (if (Class|CreativeBlueprintLibrary|IsAllowingTestPlayeronCreativeDevices)
    (bind _asfort_athena_aibot_controller (Utilities|Casting|CastToFortAthenaAIBotController Controller)
      (:then
        (switch Utilities|FlowControl|Switch|SwitchonECreativeClearItems (Variables|UserOptions|GetClearItemSetting)
          (:NewEnumerator1
            (Inventory|ClearDroppableItems _asfort_athena_aibot_controller))))
      (:CastFailed
        _))
    (else
      _))
  (bind _asfort_player_controller_athena (Utilities|Casting|CastToFortPlayerControllerAthena Controller))
  (switch Utilities|FlowControl|Switch|SwitchonECreativeClearItems (Variables|UserOptions|GetClearItemSetting)
    (:NewEnumerator1
      (Class|FortPlayerControllerAthena|ClearDroppableItems _asfort_player_controller_athena false))))

(fn GetMaxTeams
  ;[1 1 1] For simplicity we want the device to ignore the num teams during edit mode.
  (bind _logic_component (Variables|Default|GetLogic%20Component)
    (:then
      (bind _currentminigame (Variables|/Script/FortniteGame.FortMinigameLogicComponent|GetCurrentMinigame _logic_component)
        (:then
          (Utilities|IsValid (Minigame|GetMinigameTeamComponent _currentminigame)
            (:Is%20Valid
              (return (or (>= (Minigame|Teams|GetMinigameNumTeams (Minigame|GetMinigameTeamComponent _currentminigame)) 2) (or (== (Minigame|Teams|GetMinigameNumTeams (Minigame|GetMinigameTeamComponent _currentminigame)) -1) (not (Minigame|IsinPlayMode _currentminigame :!class /Script/FortniteGame.FortMinigame)))) (select (or (== (Minigame|Teams|GetMinigameNumTeams (Minigame|GetMinigameTeamComponent _currentminigame)) -1) (not (Minigame|IsinPlayMode _currentminigame :!class /Script/FortniteGame.FortMinigame))) (Getters|GetMaximumTeams) (Minigame|Teams|GetMinigameNumTeams (Minigame|GetMinigameTeamComponent _currentminigame)))))
            (:Is%20Not%20Valid
              _)))
        (:else
          _))
      (return false (Getters|GetMaximumTeams)))
    (:else
      _)))

(fn GetPlayerCanChoose
  ;[1 1 1] Easy access getter for blueprint cleanliness.
  (return (Utilities|Enum|Equal(Enum) (Variables|Internal|GetTeamSelectorState) "NewEnumerator1")))

(fn GetRandomTeam
  (bind (_valid_number_of_teams _output_team) (Getters|GetMaxTeams))
  (if _valid_number_of_teams
    (return (Math|Random|RandomIntegerinRange 0 (- _output_team 1)))
    (else
      (return -1))))

(fn GetTeam
  (return (Variables|Internal|GetTeamSelectorState) (Variables|Internal|GetTeamInt)))

(fn GetTeamName
  (params (Team%20Index :type Integer))
  (bind _logic_component (Variables|Default|GetLogic%20Component)
    (:then
      (bind _currentminigame (Variables|/Script/FortniteGame.FortMinigameLogicComponent|GetCurrentMinigame _logic_component)
        (:then
          (Utilities|IsValid (Minigame|GetMinigameTeamComponent _currentminigame)
            (:Is%20Valid
              (return (Minigame|GameRules|GetTeamName (Minigame|GetMinigameTeamComponent _currentminigame) (Teams|IndextoTeam Team%20Index))))
            (:Is%20Not%20Valid
              _)))
        (:else
          _))
      (return (Utilities|Text|FormatText "NSLOCTEXT(\"[2E3642E648D8C9239BDB248C9233DEB4]\", \"371263F447FD419C184114B5F541C78F\", \"Team {index}\")" (+ Team%20Index 1))))
    (:else
      _)))

(fn GetMaximumTeams
  (return (select (Variables|ComponentSettings|GetbUseOldKnobValues) (Variables|ComponentSettings|GetMaximum%20Teams) (Playlist|GetMaxNumberOfTeams self))))

(fn InitializeComponent
  (params (Logic%20Component :type Fort%20Minigame%20Logic%20Component%20Object%20Reference) (Team%20Index%20Setting :type Integer) (ClearItemSetting :type ECreativeClearItems%20Enum :default "NewEnumerator0"))
  (Variables|Default|SetLogic%20Component Logic%20Component)
  (Variables|UserOptions|SetClearItemSetting ClearItemSetting)
  (Initialization|InitializeTeamSelectorState Team%20Index%20Setting)
  (bind _success (Initialization|BindDelegates true))
  ;[1 1 1] Only attempt to set Team if the delegates bound correctly.
  ; (If they didn't, this logic will be called On Added to Minigame)
  (if _success
    (CallFunction|AttemptSetTeam (Variables|Internal|GetDesiredTeamInt))))

(fn InitializeTeamSelectorState
  (params (Team%20Index%20Setting :type Integer))
  (Utilities|FlowControl|Sequence
    (:then_0
      (CallFunction|SetTeamSelectorState Team%20Index%20Setting))
    (:then_1
      (Variables|Internal|SetDesiredTeamInt (select (Variables|ComponentSettings|GetbUseOldKnobValues) Team%20Index%20Setting (Knobs|Team|GetTeamValueasIndex Team%20Index%20Setting))) ; Save a copy of the team we want.\nThis is so that we can reset later if team number change\nor the minigame wasn't initialized when we started\n
      (return))))

(fn OnRep_TeamInt
  (Default|CallOnDeviceTeamUpdated (Variables|Internal|GetTeamSelectorState) (Variables|Internal|GetTeamInt))
  (Utilities|SwitchHasAuthority
    (:Authority
      (if (Variables|ComponentSettings|GetSet%20Building%20Team%20on%20Switch)
        (bind _asbuilding_actor (Utilities|Casting|CastToBuildingActor (Components|GetOwner)))
        (Class|BuildingActor|SetTeam _asbuilding_actor (Teams|IndextoTeam (Variables|Internal|GetTeamInt)))))))

(fn ReattemptPost-MinigameAddInitialization
  (params (Minigame :type Fort%20Minigame%20Object%20Reference))
  ;[1 1 1] Calling Bind Delegates again now that we've been added to the minigame. This will only be called if the initialization of the device was activated prior
  (Minigame|UnbindEventfromOnAddedtoMinigame (Variables|Default|GetLogic%20Component) (EventDispatchers|CreateEvent :!event Reattempt%20Post-Minigame%20Add%20Initialization)) ; We won't need this bind anymore, unbind it
  (Initialization|BindDelegates true)
  (CallFunction|AttemptSetTeam (Variables|Internal|GetDesiredTeamInt)))

(fn SetTeamonPlayModeChange
  (params (Minigame :type Fort%20Minigame%20Object%20Reference) (bIsInPlayMode :type Boolean))
  (if bIsInPlayMode
    (if (Utilities|Enum|Equal(Enum) (Variables|Internal|GetTeamSelectorState) "NewEnumerator1")
      (CallFunction|AttemptSetTeam (Getters|GetRandomTeam)) ; Initialize to a random team if we're in play mode and allowing player choice.
      (else
        _))
    (else
      _))
  (CallFunction|AttemptSetTeam (Variables|Internal|GetDesiredTeamInt)))

(fn SetTeamSelectorState
  (params (InTeamValue :type Integer))
  (if (Variables|ComponentSettings|GetbUseOldKnobValues)
    ;[1 1 1] Team Selector Knob is a team index with special cases for negative numbers as to avoid problems in the future with expansion. -2 = Player Choice, -1 = Random, 0+ = That Team
    (switch int InTeamValue
      (:Default
        (Variables|Internal|SetTeamSelectorState "NewEnumerator3"))
      (:-3
        (Variables|Internal|SetTeamSelectorState "NewEnumerator0"))
      (:-2
        (Variables|Internal|SetTeamSelectorState "NewEnumerator1"))
      (:-1
        (Variables|Internal|SetTeamSelectorState "NewEnumerator2")))
    (elif (Knobs|Team|IsTeamDoNotSwitch (Variables|GetInTeamValue))
      (Variables|Internal|SetTeamSelectorState "NewEnumerator0")
      (elif (Knobs|Team|IsTeamRandom (Variables|GetInTeamValue))
        (Variables|Internal|SetTeamSelectorState "NewEnumerator2")
        (else
          (Variables|Internal|SetTeamSelectorState "NewEnumerator3"))))))
