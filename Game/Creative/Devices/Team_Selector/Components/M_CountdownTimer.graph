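# Translucent, unlit, two-sided hologram material for the Team Selector device's countdown timer, built on a military numbers texture.
# The texture's red channel, split by a threshold, blends inner and outer color and opacity, with scanlines modulating the opacity.
# World position offset math combines hologram wiggle with a camera-facing basis transform that ignores camera height.
# A second camera-yaw RotateAboutAxis chain around the actor position is also present.
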Material /Game/Creative/Devices/Team_Selector/Components/M_CountdownTimer
graph {
 properties {
  bEnableResponsiveAA=true
  blendMode=BLEND_Translucent
  pixelDepthOffsetMode=PDOM_Legacy
  shadingModel=MSM_Unlit
  translucencyLightingMode=TLM_VolumetricPerVertexDirectional
  twoSided=true
  usedShadingModels=Unlit
 }
 nodes [
  n1 = TextureSampleParameter2D() {group=Color, parameterName=Texture, texture=/Game/Environments/Props/Military/Textures/T_MilitaryNumbers_01_D}
  Add(A=@n1.RGB) {constB=0}
  n2 = MaterialFunctionCall!(
   WPO Intensity A=VectorParameter() {parameterName=WPO Intensity A, defaultValue=<rgba 0.5028 0.4528 0.5586 0>}.RGB
   WPO Intensity B=VectorParameter() {parameterName=WPO Intensity B, defaultValue=<rgba 0 2 0 0>}.RGB
   Time
   Bias
  ) {materialFunction=/Game/Effects/Fort_Effects/Effects/Atlas/Materials/MF_Atlas_HologramWiggle_01}
  n3 = Subtract(A=WorldPosition().XYZ, B=NamedRerouteUsage())
  n4 = Subtract(A=CameraPositionWS(), B=NamedRerouteUsage())
  n5 = MaterialFunctionCall(Vector=Multiply(
    A=Transform(None=@n4) {transformSourceType=TRANSFORMSOURCE_World, transformType=TRANSFORM_Local}
    B=Constant3Vector() {constant=<rgba 1 1 0 0>}
   )) {materialFunction=&SafeNormalize}
  n6 = MaterialFunctionCall(Vector=VectorParameter() {parameterName=New UP, defaultValue=<rgba 0 0 1 0>}.RGB) {materialFunction=&SafeNormalize}
  n7 = Multiply(A=MaterialFunctionCall(Vector=CrossProduct(A=@n5.Result, B=@n6.Result)) {materialFunction=&SafeNormalize}.Result) {constB=-1}
  n8 = MaterialFunctionCall(
   VectorToTransform=Transform(None=@n3) {transformSourceType=TRANSFORMSOURCE_World, transformType=TRANSFORM_Local}
   BasisX=@n5.Result
   BasisY=@n7
   BasisZ=CrossProduct(A=@n5.Result, B=@n7)
  ) {materialFunction=&Transform3x3Matrix}
  Add(A=@n2.Result, B=Subtract(A=Transform(None=@n8.Transformed Vector) {transformSourceType=TRANSFORMSOURCE_Local}, B=@n3))
  n9 = ScalarParameter() {defaultValue=1, parameterName=Opacity Outer}
  n10 = ScalarParameter() {defaultValue=-1, parameterName=OpacityInner}
  n11 = Multiply(A=@n1.R, B=ScalarParameter() {defaultValue=1, group=Color, parameterName=OuterMultiplier})
  n12 = If(
   A=@n1.R
   "A > B"=Multiply(A=@n1.R, B=ScalarParameter() {defaultValue=1, group=Color, parameterName=InnerMultiplier})
   "A == B"=@n11
   "A < B"=@n11
  ) {constB=0.1}
  n13 = MaterialFunctionCall!(
   Number of Scanlines=ScalarParameter() {defaultValue=150, parameterName="Number of Scanlines (S)"}
   Scanline Speed
   "Scanlines Thickness (Bias)"
   Scanlines Dark Value
   Scanlines Light Value
   Block Opacity=@n2.Intense 0-1
   UVs
   Time
   Block Repeat
  ) {materialFunction=/Game/Effects/Fort_Effects/Effects/Atlas/Materials/MF_Atlas_Scanlines_01}
  Add(A=Multiply(A=LinearInterpolate(A=@n9, B=@n10, Alpha=@n12), B=@n13.Result), B=@n13.Secondary Lines)
  LinearInterpolate(B=VectorParameter() {parameterName=ColorInner, defaultValue=<rgba 2.2378 1.7014 20 0>}.RGB, Alpha=@n12)
  n14 = VectorParameter() {parameterName=RightVector, defaultValue=<rgba 0 1 0 1>}
  n15 = ActorPositionWS()
  n16 = CameraPositionWS()
  n17 = Multiply(A=Normalize(VectorInput=Subtract(A=@n15, B=@n16)), B=Constant3Vector() {constant=<rgba 1 1 0 0>})
  n18 = DotProduct(A=Normalize(VectorInput=Transform(None=@n14.RGB) {transformSourceType=TRANSFORMSOURCE_Local}), B=@n17)
  n19 = VectorParameter() {parameterName=ForwardVector, defaultValue=<rgba 1 0 0 1>}
  n20 = DotProduct(A=@n17, B=Normalize(VectorInput=Transform(None=@n19.RGB) {transformSourceType=TRANSFORMSOURCE_Local}))
  n21 = MaterialFunctionCall() {materialFunction=&Pi}
  n22 = Divide(A=Add(A=Arctangent2Fast(Y=@n18, X=@n20), B=@n21.Result), B=@n21.Result)
  Multiply(
   A=RotateAboutAxis(
    NormalizedRotationAxis=Constant3Vector() {constant=<rgba 0 0 1 0>}
    RotationAngle=Divide(A=@n22)
    PivotPoint=ActorPositionWS()
    Position=WorldPosition().XYZ
   )
   B=ScalarParameter() {defaultValue=1, parameterName=RotateWithCamera}
  )
  n23 = VectorParameter() {parameterName=OriginPosition}
  NamedRerouteDeclaration(None=@n23.RGB) {name=Origin Position, nodeColor=<rgba 1 0.3984 0 1>}
  NamedRerouteDeclaration(None=@n23.A) {name=Origin Alpha, nodeColor=<rgba 1 0 0.5625 1>}
  NamedRerouteUsage()
  VectorParameter() {bUseCustomPrimitiveData=true, parameterName=POS}
 ]
}
